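# Material function for the slush machine prop that adds animated slush and a shaking motion, toggled by a static switch.
# Scrolling Perlin noise distorts rounded gradients into two liquid layers with surface lines, using bump offset for depth and custom primitive data for fill level.
# A radial mask blends the slush into base color and emissive.
# Bounds-masked sine jitter and axis rotation drive WPO, scaled by custom primitive data.
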MaterialFunction /WildEstateTech/Environment/Props/WildEstate/WildEstate_SlushMachine_A/Materials/MF_WildEstate_SlushMachine
graph {nodes [
  n1 = Multiply(A=NamedRerouteUsage()) {constB=0.3}
  n2 = NamedRerouteUsage()
  n3 = ScalarParameter() {defaultValue=0.05, group=Slush Machine, parameterName=SlushMachine.Slush.BackgroundLayerOffset}
  n4 = TextureCoordinate()
  n5 = Multiply(A=Time(), B=ScalarParameter() {defaultValue=0.05, group=Slush Machine, parameterName=SlushMachine.Slush.NoiseSpeed})
  n6 = TextureSample(UVs=Multiply(A=Add(A=@n4, B=@n5), B=ScalarParameter() {defaultValue=3, group=Slush Machine, parameterName=SlushMachine.Slush.NoiseTiling})) {texture=/Game/Global/Textures/Default/Noise/3xPerlin_M}
  n7 = ScalarParameter() {defaultValue=0.02, group=Slush Machine, parameterName=SlushMachine.Slush.NoiseIntensityForeground}
  n8 = ScalarParameter() {defaultValue=-0.369, group=Slush Machine, parameterName=SlushMachine.Slush.FillMin}
  n9 = ScalarParameter() {defaultValue=-0.15, group=Slush Machine, parameterName=SlushMachine.Slush.FillMax}
  n10 = ScalarParameter() {
   bUseCustomPrimitiveData=true
   desc=master control for the Shaking for the Slush Mashine WPO
   group=Slush Machine
   parameterName=BP_SlushMachineFillAlpha
   primitiveDataIndex=1
   sortPriority=40
  }
  n11 = Saturate(None=@n10)
  n12 = Add(A=TextureCoordinate(), B=AppendVector(A=Constant(), B=LinearInterpolate(A=@n8, B=@n9, Alpha=@n11)))
  n13 = BumpOffset(
   Coordinate=Add(A=@n12, B=Multiply(A=@n6.G, B=@n7) {constB=0.05})
   Height=ScalarParameter() {defaultValue=-1.5, group=Slush Machine, parameterName=SlushMachine.Slush.BumpOffset.Height}
  )
  n14 = AppendVector(
   A=Constant()
   B=Multiply(A=ScalarParameter() {defaultValue=0.02, group=Slush Machine, parameterName=SlushMachine.Slush.LineThickness}) {constB=0.1}
  )
  n15 = Add(
   A=@n3
   B=MaterialFunctionCall(UV Channel=Add(A=@n13, B=@n14)) {materialFunction=&LinearGradient}.VGradient
  ) {constB=0.05}
  n16 = Subtract(A=Saturate(None=Round(None=@n15)), B=NamedRerouteUsage())
  n17 = NamedRerouteUsage()
  n18 = LinearInterpolate(
   A=Multiply(
    A=LinearInterpolate(A=@n1, Alpha=@n2)
    B=VectorParameter() {group=Slush Machine, parameterName=SlushMachine.Slush.Color, defaultValue=<rgba 1 0 0 1>}.RGB
   )
   Alpha=Saturate(None=Subtract(A=@n16, B=@n17))
  )
  n19 = Add(
   A=@n12
   B=Multiply(A=ScalarParameter() {defaultValue=0.05, group=Slush Machine, parameterName=SlushMachine.Slush.NoiseIntensityBackground}, B=@n6.R) {constB=0.1}
  )
  n20 = MaterialFunctionCall(UV Channel=Add(A=@n19, B=@n14)) {materialFunction=&LinearGradient}
  n21 = Subtract(A=Saturate(None=Round(None=@n20.VGradient)), B=NamedRerouteUsage())
  FunctionOutput(None=NamedRerouteDeclaration(None=LinearInterpolate(A=@n18, Alpha=@n21)) {name=Color, nodeColor=<rgba 0 0.1328 1 1>}) {outputName=Color, sortPriority=1}
  n22 = GetMaterialAttributes(None=NamedRerouteDeclaration(None=FunctionInput() {bUsePreviewValueAsDefault=true, inputName=MaterialIn, inputType=FunctionInput_MaterialAttributes}) {name=MatIn, nodeColor=<rgba 1 0.1875 0 1>}) {attributeGetTypes [#1, #2]}
  n23 = NamedRerouteUsage()
  n24 = VectorParameter() {
   group=Slush Machine
   parameterName=SlushMachine.Slush.AreaMaskParams
   channelNames {a=Density, b=Radius, g=RadialMaskPosY, r=RadialMaskPosX}
   defaultValue=<rgba 0.5 0.5 0.1 100>
  }
  n25 = MaterialFunctionCall(CenterPosition=AppendVector(A=@n24.RadialMaskPosX, B=@n24.RadialMaskPosX), Radius=@n24.Radius, Density=@n24.Radius) {materialFunction=&RadialGradientExponential}
  n26 = LinearInterpolate(A=@n22.BaseColor, B=@n23, Alpha=@n25.RadialGradientExponential)
  n27 = LinearInterpolate(A=@n22.EmissiveColor, B=@n23, Alpha=@n25.RadialGradientExponential)
  n28 = Multiply(A=Time(), B=ScalarParameter() {defaultValue=10, group=Slush Machine, parameterName=SlushMachine.WPO.JitterSpeedVert, sortPriority=40}) {constB=10}
  n29 = Multiply(
   A=Multiply(
    A=Sine(None=@n28)
    B=ScalarParameter() {defaultValue=0.01, group=Slush Machine, parameterName=SlushMachine.WPO.JitterDistanceVert, sortPriority=40}
   ) {constB=0.01}
   B=Constant3Vector() {constant=<rgba 0 0 1 0>}
  )
  n30 = VectorParameter() {group=Slush Machine, parameterName=SlushMachine.WPO.Axis, sortPriority=40, defaultValue=<rgba 1 0 0 1>}
  n31 = Sine(None=Multiply(A=Time(), B=ScalarParameter() {defaultValue=10, group=Slush Machine, parameterName=SlushMachine.WPO.JitterSpeed, sortPriority=40}) {constB=10})
  n32 = ScalarParameter() {defaultValue=0.01, group=Slush Machine, parameterName=SlushMachine.WPO.JitterDistance, sortPriority=40}
  n33 = Sine(None=Time())
  n34 = LinearInterpolate(Alpha=SmoothStep(
    Min=ScalarParameter() {defaultValue=0.3, group=Slush Machine, parameterName=SlushMachine.WPO.SmoothStepMin, sliderMax=0.5, sortPriority=40}
    Max=ScalarParameter() {defaultValue=0.6, group=Slush Machine, parameterName=SlushMachine.WPO.SmoothStepMax, sliderMax=0.5, sortPriority=40}
    Value=ConstantBiasScale(None=@n33)
   ) {constMax=0.6, constMin=0.3}) {constA=-1}
  n35 = ScalarParameter() {defaultValue=0.03, group=Slush Machine, parameterName=SlushMachine.WPO.Distance, sortPriority=40}
  n36 = VectorParameter() {group=Slush Machine, parameterName=SlushMachine.WPO.PivotOffset, sortPriority=40, defaultValue=<rgba 0 0 0 1>}
  n37 = RotateAboutAxis(
   NormalizedRotationAxis=Normalize(VectorInput=Transform(None=@n30.RGB) {transformSourceType=TRANSFORMSOURCE_Local})
   RotationAngle=Add(A=Multiply(A=@n31, B=@n32) {constB=0.01}, B=Multiply(A=@n34, B=@n35))
   PivotPoint=TransformPosition(None=@n36.RGB) {transformType=TRANSFORMPOSSOURCE_World}
   Position=WorldPosition().XYZ
  )
  n38 = ScalarParameter() {group=Slush Machine, parameterName=SlushMachine.WPO.MaskMin}
  n39 = ScalarParameter() {defaultValue=1, group=Slush Machine, parameterName=SlushMachine.WPO.MaskMax}
  n40 = Saturate(None=MaterialFunctionCall() {materialFunction=&BoundingBoxBased_0-1_UVW}.B)
  n41 = ScalarParameter() {
   bUseCustomPrimitiveData=true
   desc=master control for the Shaking for the Slush Mashine WPO
   group=Slush Machine
   parameterName=BP_SlushMachineWPOIntensity
   primitiveDataIndex=2
   sortPriority=40
  }
  n42 = Multiply(A=LinearInterpolate(B=Add(A=@n29, B=@n37), Alpha=SmoothStep(Min=@n38, Max=@n39, Value=@n40)), B=Saturate(None=@n41))
  FunctionOutput(None=StaticSwitchParameter(
    True=SetMaterialAttributes!(MaterialAttributes=@n22.MaterialAttributes, BaseColor=@n26, EmissiveColor=@n27, WorldPositionOffset=@n42) {attributeSetTypes [#1, #2, #3]}
    False=NamedRerouteUsage()
   ) {group=Wild Estate Tech Art FX, parameterName=UseSlushMachineFX, sortPriority=1}) {outputName=Material}
  FunctionOutput(None=@n42) {outputName=Shake WPO, sortPriority=4}
  n43 = MaterialFunctionCall(UV Channel=@n19) {materialFunction=&LinearGradient}
  NamedRerouteDeclaration(None=Saturate(None=Round(None=@n43.VGradient))) {name=ForegroundMask, nodeColor=<rgba 1 0.0415 0.8158 1>}
  n44 = Add(
   A=MaterialFunctionCall(UV Channel=@n13) {materialFunction=&LinearGradient}.VGradient
   B=@n3
  ) {constB=0.05}
  NamedRerouteDeclaration(None=Saturate(None=Round(None=@n44))) {name=BackgroundMask, nodeColor=<rgba 1 0 0.5859 1>}
  Sine()
  StaticSwitchParameter(True=@n10, False=ScalarParameter() {defaultValue=0.75, group=Slush Machine, parameterName=SlushMachine.Slush.FillAlpha, sliderMax=1}) {group=Slush Machine, parameterName=Use Custom Prim ForIntensity And Fill Level, sortPriority=1}
 ]}
