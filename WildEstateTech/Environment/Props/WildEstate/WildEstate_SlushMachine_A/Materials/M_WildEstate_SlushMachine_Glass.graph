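# Additive, unlit, two-sided glass material for the Wild Estate slush machine, with fake reflections.
# A window texture, offset by instance position and camera direction in tangent space, supplies reflection lines for colour and opacity.
# Time-of-day glass and reflection colours are shaded against the sun vector with a darkness floor, then corrected for eye adaptation.
# A static switch picks single- or dual-flavor slush machine functions to drive shake World Position Offset.
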Material /WildEstateTech/Environment/Props/WildEstate/WildEstate_SlushMachine_A/Materials/M_WildEstate_SlushMachine_Glass
graph {
 properties {
  bAutomaticallySetUsageInEditor=false
  bUseMaterialAttributes=true
  bUsedWithInstancedStaticMeshes=true
  blendMode=BLEND_Additive
  shadingModel=MSM_Unlit
  twoSided=true
  usedShadingModels=Unlit
 }
 nodes [
  NamedRerouteDeclaration(None=LinearInterpolate(
    A=ScalarParameter() {defaultValue=1, parameterName=Opacity_Lines}
    B=ScalarParameter() {defaultValue=1, parameterName=Opacity_Glass}
    Alpha=NamedRerouteUsage()
   )) {name=Opacity, nodeColor=<rgba 0.4531 1 0 1>}
  n1 = TransformPosition(None=Constant3Vector()) {transformSourceType=TRANSFORMPOSSOURCE_Instance, transformType=TRANSFORMPOSSOURCE_World}
  n2 = MaterialFunctionCall(Float3=Divide(A=@n1) {constB=100}) {materialFunction=&BreakOutFloat3Components}
  n3 = Add(A=Add(A=@n2.R, B=@n2.R), B=@n2.B)
  n4 = Add(
   A=Multiply(A=Frac(None=@n3)) {constB=33.3}
   B=Multiply(A=TextureCoordinate(), B=ScalarParameter() {defaultValue=0.5, parameterName=GlassRelectionTiling})
  )
  n5 = Normalize(VectorInput=Subtract(A=@n1, B=CameraPositionWS()))
  n6 = Multiply(
   A=ComponentMask(None=Transform(None=@n5) {transformSourceType=TRANSFORMSOURCE_World, transformType=TRANSFORM_Tangent}) {g=true, r=true}
   B=ScalarParameter() {defaultValue=0.1, parameterName=GlassRelectionOffset}
  )
  NamedRerouteDeclaration(None=TextureSample(UVs=Add(A=@n4, B=@n6)) {texture=/WildEstateTech/Environment/Textures/T_WildEstate_Windows}.R) {name=WindowLines, nodeColor=<rgba 1 0 0.7734 1>}
  n7 = VectorParameter() {parameterName=Reflection_Color_Morning, defaultValue=<rgba 0.724 0.6384 0.4014 1>}
  n8 = VectorParameter() {parameterName=Reflection_Color_Day, defaultValue=<rgba 0.1131 0.724 0.6119 1>}
  n9 = VectorParameter() {parameterName=Reflection_Color_Evening, defaultValue=<rgba 0.1131 0.724 0.6119 1>}
  n10 = VectorParameter() {parameterName=Reflection_Color_Night, defaultValue=<rgba 0.0254 0.1028 0.4427 1>}
  n11 = NamedRerouteUsage()
  n12 = Add(
   A=MaterialFunctionCall(
    Color_Morning=VectorParameter() {parameterName=Glass_Color_Morning, defaultValue=<rgba 0.8229 0.4438 0.1029 1>}.RGB
    Color_Day=VectorParameter() {parameterName=Glass_Color_Day, defaultValue=<rgba 0.1131 0.724 0.6119 1>}.RGB
    Color_Evening=VectorParameter() {parameterName=Glass_Color_Evening, defaultValue=<rgba 0.724 0.381 0.131 1>}.RGB
    Color_Night=VectorParameter() {parameterName=Glass_Color_Night, defaultValue=<rgba 0.0407 0.1246 0.224 1>}.RGB
   ) {materialFunction=&Color_Time_of_Day}.Result
   B=Multiply(
    A=MaterialFunctionCall(Color_Morning=@n7.RGB, Color_Day=@n8.RGB, Color_Evening=@n9.RGB, Color_Night=@n10.RGB) {materialFunction=&Color_Time_of_Day}.Result
    B=OneMinus(None=@n11)
   )
  )
  n13 = CollectionParameter() {
   collection=&FortniteMaterialParameters
   parameterName=SunlightDirectionalVector
  }
  n14 = Multiply(A=VertexNormalWS(), B=TwoSidedSign())
  n15 = Saturate(None=Max(A=DotProduct(A=@n13, B=@n14), B=ScalarParameter() {defaultValue=0.4109, parameterName=MinDarkness}))
  NamedRerouteDeclaration(None=EyeAdaptationInverse(LightValueInput=Multiply(A=@n12, B=@n15)).EyeAdaptationInverse) {name=Emissive, nodeColor=<rgba 0 1 0.9063 1>}
  SetMaterialAttributes!(
   MaterialAttributes
   Emissive Color=NamedRerouteUsage()
   Opacity=NamedRerouteUsage()
   World Position Offset=StaticSwitchParameter(
    True=MaterialFunctionCall!(MaterialIn) {materialFunction=/WildEstateTech/Environment/BaseMaterials/MaterialFunctions/MF_SlushieMachine_DualFlavor}.Shake WPO
    False=MaterialFunctionCall!(MaterialIn) {materialFunction=&/MF_WildEstate_SlushMachine}.Shake WPO
   ) {group=Wild Estate Tech Art FX, parameterName=UseDualFlavorSlushMF}
  ) {attributeSetTypes [#1, #2, #3]}
 ]
}
